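# Sound cue for a single Week 10 Battle Pass Season 20 voice line spoken by The Origin.
# A lone dialogue player node plays the line's dialogue wave in a context with The Origin as speaker and no targets.
# The cue routes through the narrator VO sound class with a raised volume multiplier and a very large max distance.
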SoundCue /BattlePassS20/Audio/VO/Cues/Week_10/02/ORIGIN_2022_REGS_Week10_02_03_010_Dialogue_Cue
root=DialoguePlayer_0
nodes [{
  id=DialoguePlayer_0
  properties {dialogueWaveParameter {
    dialogueWave=/BattlePassS20/Audio/VO/Waves/Origin/ORIGIN_2022_REGS_Week10_02_03_010_Dialogue
    context {speaker=/BattlePassS20/Audio/VO/TheOrigin_DV, targets []}
   }}
 }]
properties {maxDistance=2097152, soundClassObject=/Game/Sounds/FortSoundClassesAndModes/Fort_VO_Narrator, volumeMultiplier=1.3333}
